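; EQS context for mid-match assigned environmental item spawning that provides the white circle's center, meaning the next safe zone center, as a single location.
; ProvideSingleLocation gets the global safe zone interface and returns its next center, ignoring the querier inputs.
; A second Sequence branch returns the zero vector, but the first branch always returns before it can run.
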(fn ProvideSingleLocation
  (params (QuerierObject :type Object%20Reference) (QuerierActor :type Actor%20Object%20Reference))
  (Utilities|FlowControl|Sequence
    (:then_0
      (return (SafeZoneInterface|GetSafeZoneNextCenter (SafeZoneInterface|GetGlobalSafeZoneInterface))))
    (:then_1
      (return "0, 0, 0"))))
