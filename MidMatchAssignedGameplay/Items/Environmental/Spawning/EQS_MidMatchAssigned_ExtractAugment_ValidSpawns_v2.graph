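# EQS query that picks valid Extract Augment spawn locations for MidMatchAssigned gameplay.
# It traces a grid around the white-circle safe zone center onto geometry.
# Distance, overlap and surface tests filter points, with spacing from spawned objectives also scored.
# Default negative weights favor points near squads and the next circle center, with small squad and closest-player distance deltas, tunable through bound scalable floats.
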EnvQuery /MidMatchAssignedGameplay/Items/Environmental/Spawning/EQS_MidMatchAssigned_ExtractAugment_ValidSpawns_v2
options [{
  generator {
   name=EnvQueryGenerator_SimpleGrid_0
   class=EnvQueryGenerator_SimpleGrid
   properties {
    generateAround=&/EQSContext_MMAssigned_WhiteCircleSafeZoneCenter.EQSContext_MMAssigned_WhiteCircleSafeZoneCenter_C
    gridSize {
     dataBinding=&self:EnvQueryGenerator_SimpleGrid_0.AIDataProvider_QueryParams_0
     dataField=FloatValue
     defaultValue=45000
    }
    projectionData {
     extentX=0
     navigationFilter=null
     postProjectionVerticalOffset=0
     projectDown=30000
     projectUp=30000
     serializedChannel=ECC_GameTraceChannel6
     traceChannel=TraceTypeQuery7
     traceMode=GeometryByChannel
     traceShape=Sphere
    }
    spaceBetween {
     dataBinding=&self:EnvQueryGenerator_SimpleGrid_0.FortAIDataProvider_ScalableFloat_1
     dataField=FloatValue
     defaultValue=3000
    }
   }
  }
  tests [
   EnvQueryTest_Distance_0 {
    class=EnvQueryTest_Distance
    properties {
     distanceTo=&EQSContext_MMAssigned_WhiteCircleSafeZoneCenter_C
     filterType=Maximum
     scoringEquation=Linear
     scoringFactor=1
     testComment=""
     testMode=Distance2D
     testPurpose=Filter
     floatValueMax {
      dataBinding=&self:EnvQueryTest_Distance_0.AIDataProvider_QueryParams_1
      dataField=FloatValue
      defaultValue=0
     }
     floatValueMin {
      dataBinding=&self:EnvQueryTest_Distance_0.AIDataProvider_QueryParams_0
      dataField=FloatValue
      defaultValue=0
     }
    }
   }
   EnvQueryTest_Overlap_2 {
    class=EnvQueryTest_Overlap
    properties {overlapData {extentX=250, extentY=250, extentZ=250, overlapChannel=ECC_GameTraceChannel1, shapeOffset=<0 0 300>}}
   }
   EnvQueryTest_Overlap_3 {
    class=EnvQueryTest_Overlap
    properties {overlapData {bOnlyBlockingHits=false, extentX=300, extentY=300, extentZ=300, overlapChannel=ECC_GameTraceChannel15, shapeOffset=<0 0 0>}}
   }
   EnvQueryTest_Distance_1 {
    class=EnvQueryTest_Distance
    properties {
     distanceTo=/Script/MidMatchAssignedGameplayRuntime.FortQueryContext_SpawnedObjectiveLocations
     filterType=Minimum
     scoringEquation=Linear
     scoringFactor=1
     testComment=""
     testMode=Distance2D
     testPurpose=FilterAndScore
     floatValueMin {
      dataBinding=&self:EnvQueryTest_Distance_1.FortAIDataProvider_ScalableFloat_0
      dataField=FloatValue
      defaultValue=0
     }
    }
   }
   FortQueryTest_ValidSurface_1 {
    class=FortQueryTest_ValidSurface
    properties {
     traceOffsetDown=100
     flatSurfaceToleranceZ {dataBinding=null, dataField=FloatValue, defaultValue=50}
     invalidHitActorClasses [
      /Script/FortniteGame.BuildingDeco
      &BuildingProp
      /Script/FortniteGame.BuildingGameplayActor
      &BuildingStairs
      &BuildingWall
     ]
     radius {dataBinding=null, dataField=FloatValue, defaultValue=300}
     surfaceMaterials []
     traceOffsetUp {dataBinding=null, dataField=FloatValue, defaultValue=100}
     validHitActorClasses [
      /Script/Landscape.LandscapeStreamingProxy
      /Script/Landscape
      /Script/FortniteGame.BuildingActor
      &StaticMeshActor
      /Script/Engine.SplineMeshActor
     ]
    }
   }
   EnvQueryTest_Distance_2 {
    class=EnvQueryTest_Distance
    properties {
     distanceTo=/Script/MidMatchAssignedGameplayRuntime.FortQueryContext_SquadAvgLocationsForObjective
     filterType=Minimum
     floatValueMax=50000
     floatValueMin=0
     multipleContextScoreOp=MaxScore
     scoringEquation=Linear
     testComment=""
     testMode=Distance2D
     testPurpose=Score
     scoringFactor {
      dataBinding=&self:EnvQueryTest_Distance_2.FortAIDataProvider_ScalableFloat_0
      dataField=FloatValue
      defaultValue=-1
     }
    }
   }
   FortQueryTest_DeltaDistance_0 {
    class=FortQueryTest_DeltaDistance
    properties {
     bDefineReferenceValue=false
     bUseDistance2D=true
     boolValue=true
     clampMaxType=None
     clampMinType=None
     filterType=Range
     floatValueMax=0
     locationProviderContext=&FortQueryContext_SquadAvgLocationsForObjective
     multipleContextFilterOp=AllPass
     multipleContextScoreOp=AverageScore
     normalizationType=Absolute
     referenceValue=0
     scoreClampMax=0
     scoreClampMin=0
     scoringEquation=Linear
     testComment=""
     testPurpose=Score
     floatValueMin {dataBinding=null, dataField=FloatValue, defaultValue=0}
     scoringFactor {
      dataBinding=&self:FortQueryTest_DeltaDistance_0.FortAIDataProvider_ScalableFloat_3
      dataField=FloatValue
      defaultValue=-1
     }
    }
   }
   EnvQueryTest_Distance_3 {
    class=EnvQueryTest_Distance
    properties {
     distanceTo=&/EQSContext_MMAssigned_NextWhiteCircleSafeZoneCenter.EQSContext_MMAssigned_NextWhiteCircleSafeZoneCenter_C
     filterType=Range
     floatValueMin=0
     scoringEquation=Linear
     testComment=""
     testMode=Distance2D
     testPurpose=Score
     scoringFactor {
      dataBinding=&self:EnvQueryTest_Distance_3.FortAIDataProvider_ScalableFloat_0
      dataField=FloatValue
      defaultValue=-0.2
     }
    }
   }
   FortQueryTest_ClosestPlayersToObjectiveDeltaDistance_0 {
    class=FortQueryTest_ClosestPlayersToObjectiveDeltaDistance
    properties {
     bDefineReferenceValue=false
     bUseDistance2D=false
     boolValue=true
     clampMaxType=None
     clampMinType=None
     filterType=Range
     floatValueMax=0
     floatValueMin=0
     multipleContextFilterOp=AllPass
     multipleContextScoreOp=AverageScore
     normalizationType=Absolute
     referenceValue=0
     scoreClampMax=0
     scoreClampMin=0
     scoringEquation=Linear
     testComment=""
     testPurpose=Score
     scoringFactor {
      dataBinding=&self:FortQueryTest_ClosestPlayersToObjectiveDeltaDistance_0.FortAIDataProvider_ScalableFloat_0
      dataField=FloatValue
      defaultValue=-1
     }
    }
   }
  ]
 }]
