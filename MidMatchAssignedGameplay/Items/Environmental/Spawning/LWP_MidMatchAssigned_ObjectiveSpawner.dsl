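; Rift encounter spawner for a mid-match assigned objective, showing its spawn anticipation VFX and audio burst only to clients assigned to that objective.
; BeginPlay and the objective's OnRep both feed a guarded, run-once replication handler, which caches local assignment on remote clients.
; The audio burst routes through a client event, and both effects check that cached flag before calling parent logic.
; It also broadcasts spawn completion and offsets the teleport location above the actor.
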(event EventBeginPlay
  (Components|Timeline|SetPlayRate (Variables|LWP_RiftEncounter_Parent|GetDecalAnimation) (/ 1.0 0.35))
  (Utilities|IsValid (Variables|Default|GetAssociatedAssignedObjective)
    (:Is%20Valid
      (if (not (Variables|Default|GetbIsObjectiveReferenceReplicated))
        (Variables|Default|SetbIsObjectiveReferenceReplicated true)
        (CallFunction|OnObjectiveReferenceReplicated)))))

(event Custom|OnObjectiveReferenceReplicated
  (Networking|SwitchHasAuthority
    (:Authority
      _)
    (:Remote
      (Variables|Default|SetbObjectiveIsAssignedToLocalClient (Assignment|IsLocalClientAssignedtoObjective (Variables|Default|GetAssociatedAssignedObjective)))))
  (|CallParentFunction))

(event EventSpawnAudioBurst
  ;[0.4093 0 1] SpawnAudioBurst (Override)
  ; Replicate the spawn to clients so only assigned clients will see/hear the FX
  (CallFunction|SpawnAssignedAudioBurstClient))

(event Custom|SpawnAssignedAudioBurst_Client
  ;[0.4093 0 1] SpawnAssignedAudioBurst_Client
  ; Calls the parent Spawn Audio Burst logic ONLY on assigned clients
  (Networking|CanExecuteCosmeticEvents
    (:True
      (if (Variables|Default|GetbObjectiveIsAssignedToLocalClient)
        (|CallParentFunction)))))

(event EventPostSpawnEvents
  (|CallParentFunction)
  (Default|CallOnRiftSpawnComplete))

(fn GetTeleportLocation
  (bind (_returnvalue_x _returnvalue_y _returnvalue_z) (Math|Vector|BreakVector (Transformation|GetActorLocation self)))
  (return (Math|Vector|MakeVector _returnvalue_x _returnvalue_y (+ _returnvalue_z 35.0))))

(fn OnRep_AssociatedAssignedObjective
  (Utilities|IsValid (Variables|Default|GetAssociatedAssignedObjective)
    (:Is%20Valid
      (if (Variables|Default|GetbIsObjectiveReferenceReplicated)
        (Networking|SwitchHasAuthority
          (:Authority
            (Development|PrintString "RiftSpawn: OnRepAssigned HasSetAssigned TRUE" true true "(R=1.000000,G=0.000000,B=0.076154,A=1.000000)" 8.0)))
        (else
          (Variables|Default|SetbIsObjectiveReferenceReplicated true)
          (CallFunction|OnObjectiveReferenceReplicated))))))

(fn SpawnAncitipationVFX
  (if (Variables|Default|GetbObjectiveIsAssignedToLocalClient)
    (|CallParentFunction)))

(fn ConstructionScript)
